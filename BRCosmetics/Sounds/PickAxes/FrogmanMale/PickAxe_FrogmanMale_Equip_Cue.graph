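# Sound cue for the FrogmanMale pickaxe equip sound, built from swing recordings rather than dedicated equip samples.
# A mixer layers two distance branches: close attenuation plays FrogmanMale swings with random pitch and volume, while medium attenuation plays DefaultV2 distant swings with random volume only.
# In each branch, a quality-level node trims the random wave pool down to a single wave at lower settings.
# Overridden concurrency limits instances per owner, stopping the oldest.
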SoundCue /BRCosmetics/Sounds/PickAxes/FrogmanMale/PickAxe_FrogmanMale_Equip_Cue
root=Mixer_0
nodes [
 {id=Attenuation_0, children [Modulator_1], properties {attenuationSettings=/Game/Sounds/Attenuation/Fort_Weapon_Pickaxe_Close_Att}}
 {id=Attenuation_1, children [Modulator_0], properties {attenuationSettings=/Game/Sounds/Attenuation/Fort_Weapon_Pickaxe_Medium_Att}}
 {id=Mixer_0, children [Attenuation_0, Attenuation_1], properties {inputVolume [1,1]}}
 {id=Modulator_0, children [QualityLevel_1]}
 {id=Modulator_1, children [QualityLevel_0]}
 {id=Modulator_2, children [Random_0], properties {pitchMax=1, pitchMin=0.85, volumeMax=0.8, volumeMin=0.7}}
 {id=Modulator_3, children [Random_1], properties {pitchMax=1, pitchMin=0.85, volumeMax=0.8, volumeMin=0.7}}
 {id=Modulator_4, children [WavePlayer_18], properties {pitchMax=1, pitchMin=0.85, volumeMax=0.8, volumeMin=0.7}}
 {id=Modulator_5, children [Random_5], properties {volumeMax=0.9, volumeMin=0.8}}
 {id=Modulator_6, children [Random_6], properties {volumeMax=0.9, volumeMin=0.8}}
 {id=Modulator_7, children [WavePlayer_37], properties {volumeMax=0.9, volumeMin=0.8}}
 {id=QualityLevel_0, children [Modulator_2,Modulator_3,Modulator_4,Modulator_4]}
 {id=QualityLevel_1, children [Modulator_5,Modulator_6,Modulator_7,Modulator_7]}
 {id=Random_0, children [WavePlayer_1,WavePlayer_2,WavePlayer_4,WavePlayer_6,WavePlayer_7,WavePlayer_8], properties {weights [1,1,1,1,1,1]}}
 {id=Random_1, children [WavePlayer_9, WavePlayer_10, WavePlayer_11], properties {weights [1,1,1]}}
 {id=Random_5, children [WavePlayer_27, WavePlayer_28, WavePlayer_29, WavePlayer_30, WavePlayer_31], properties {weights [1,1,1,1,1]}}
 {id=Random_6, children [WavePlayer_32, WavePlayer_35, WavePlayer_36], properties {weights [1,1,1]}}
 {id=WavePlayer_1, properties {soundWaveAssetPtr=&/PickAxe_FrogmanMale_Swing_01}}
 {id=WavePlayer_10, properties {soundWaveAssetPtr=&/PickAxe_FrogmanMale_Swing_03}}
 {id=WavePlayer_11, properties {soundWaveAssetPtr=&/PickAxe_FrogmanMale_Swing_04}}
 {id=WavePlayer_18, properties {soundWaveAssetPtr=&PickAxe_FrogmanMale_Swing_04}}
 {id=WavePlayer_2, properties {soundWaveAssetPtr=&/PickAxe_FrogmanMale_Swing_02}}
 {id=WavePlayer_27, properties {soundWaveAssetPtr=/BRCosmetics/Sounds/PickAxes/DefaultV2/Pickaxe_DefaultV2_Swing_Distant_01}}
 {id=WavePlayer_28, properties {soundWaveAssetPtr=/BRCosmetics/Sounds/PickAxes/DefaultV2/Pickaxe_DefaultV2_Swing_Distant_02}}
 {id=WavePlayer_29, properties {soundWaveAssetPtr=/BRCosmetics/Sounds/PickAxes/DefaultV2/Pickaxe_DefaultV2_Swing_Distant_03}}
 {id=WavePlayer_30, properties {soundWaveAssetPtr=/BRCosmetics/Sounds/PickAxes/DefaultV2/Pickaxe_DefaultV2_Swing_Distant_04}}
 {id=WavePlayer_31, properties {soundWaveAssetPtr=/BRCosmetics/Sounds/PickAxes/DefaultV2/Pickaxe_DefaultV2_Swing_Distant_05}}
 {id=WavePlayer_32, properties {soundWaveAssetPtr=&Pickaxe_DefaultV2_Swing_Distant_01}}
 {id=WavePlayer_35, properties {soundWaveAssetPtr=&Pickaxe_DefaultV2_Swing_Distant_04}}
 {id=WavePlayer_36, properties {soundWaveAssetPtr=&Pickaxe_DefaultV2_Swing_Distant_05}}
 {id=WavePlayer_37, properties {soundWaveAssetPtr=&Pickaxe_DefaultV2_Swing_Distant_04}}
 {id=WavePlayer_4, properties {soundWaveAssetPtr=&PickAxe_FrogmanMale_Swing_03}}
 {id=WavePlayer_6, properties {soundWaveAssetPtr=&PickAxe_FrogmanMale_Swing_04}}
 {id=WavePlayer_7, properties {soundWaveAssetPtr=&/PickAxe_FrogmanMale_Swing_05}}
 {id=WavePlayer_8, properties {soundWaveAssetPtr=&/PickAxe_FrogmanMale_Swing_06}}
 {id=WavePlayer_9, properties {soundWaveAssetPtr=&PickAxe_FrogmanMale_Swing_01}}
]
properties {
 bOverrideConcurrency=true
 bPrimeOnLoad=true
 maxDistance=3000
 pitchMultiplier=1.2
 soundClassObject=&Fort_WeaponMelee_PSM
 volumeMultiplier=0.2
 concurrencyOverrides {bLimitToOwner=true, maxCount=3, resolutionRule=StopOldest}
 concurrencySet [&Player_Melee_Weapon_Conc]
}
